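# Sound cue for the final impact when books are demolished, likely as the destruction sound of a medium-sized object.
# A random node picks with equal weight between two book final-impact wave variations.
# It routes through the impacts sound class with demolition medium-object attenuation at reduced volume.
# A shared object-impact concurrency set is assigned, alongside override values that cap voices and stop the oldest.
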SoundCue /Game/Sounds/Fort_Impact_Sounds/Demolition/Book/Demo_Books_ImpactFinal_Cue
root=Random_65
nodes [
 {id=Random_65, children [WavePlayer_212, WavePlayer_215], properties {weights [1,1]}}
 {id=WavePlayer_212, properties {soundWaveAssetPtr=&/Book_ImpactFinal01}}
 {id=WavePlayer_215, properties {soundWaveAssetPtr=&/Book_ImpactFinal04}}
]
properties {
 attenuationSettings=/Game/Sounds/Attenuation/Fort_Demolition_Destroy_Medium_Sized_Object_Att
 maxDistance=4000
 soundClassObject=&Fort_Impacts_PSM
 volumeMultiplier=0.544
 concurrencyOverrides {maxCount=3, resolutionRule=StopOldest}
 concurrencySet [&CONC_Object_Impacts_FourMax]
}
