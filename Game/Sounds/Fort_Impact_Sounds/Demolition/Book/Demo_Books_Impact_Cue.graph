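# Sound cue for book impact sounds in the Demolition impact set, classed under Fort_Impacts.
# A Random node picks with equal weight among three Book_Impact wave variants and feeds a Modulator at the root.
# It uses the small melee weapon attenuation and references a shared object-impact concurrency set, alongside an override that stops the oldest voice.
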SoundCue /Game/Sounds/Fort_Impact_Sounds/Demolition/Book/Demo_Books_Impact_Cue
root=Modulator_37
nodes [
 {id=Modulator_37, children [Random_64]}
 {id=Random_64, children [WavePlayer_207, WavePlayer_208, WavePlayer_210], properties {weights [1,1,1]}}
 {id=WavePlayer_207, properties {soundWaveAssetPtr=&/Book_Impact03}}
 {id=WavePlayer_208, properties {soundWaveAssetPtr=&/Book_Impact05}}
 {id=WavePlayer_210, properties {soundWaveAssetPtr=&/Book_Impact04}}
]
properties {
 attenuationSettings=&Fort_Weapon_Melee_Sml_Att
 maxDistance=2800
 soundClassObject=&Fort_Impacts
 volumeMultiplier=0.68
 concurrencyOverrides {maxCount=3, resolutionRule=StopOldest}
 concurrencySet [&CONC_Object_Impacts_FourMax]
}
